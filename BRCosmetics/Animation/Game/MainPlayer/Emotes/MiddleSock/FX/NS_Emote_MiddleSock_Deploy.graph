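# Niagara system for the deploy effect of the MiddleSock emote: a brief orange burst of sparks and flare.
# Beams bursts sprites split by partition into opposite X-axis cones, with curl noise, rising drag, partial inherited source movement and speed-based stretching.
# Round_Flare adds a large, short-lived glow with sprite and mesh renderers that swells, shrinks and fades; its dynamic material parameters module is disabled.
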NiagaraSystem /BRCosmetics/Animation/Game/MainPlayer/Emotes/MiddleSock/FX/NS_Emote_MiddleSock_Deploy
dependencies {
 usedDataInterfaces [&NiagaraDataInterfaceArrayFloat2, &NiagaraDataInterfaceCurve]
 usedDynamicInputs [&SelectVector2DFromArray]
 usedModules [
  &self:SystemSpawnScript.SystemScriptSource.SystemScriptGraph.NiagaraNodeAssignment_0.SetVariables_#1
  &SystemState
 ]
}
emitters {
 Beams {
  emitterUpdateScript [
   &EmitterState()
   &SpawnBurst_Instantaneous(Spawn Count=15, Use Spawn Probability=false)
  ]
  particleSpawnScript [
   &PartitionParticles()
   &InitializeParticle(
    Color Maximum=<rgba 1 0.1741 0 1>
    Color Minimum=<rgba 1 0.4641 0 1>
    Color Mode=::NewEnumerator2
    Lifetime Max=1
    Lifetime Min=0.3
    Lifetime Mode=::NewEnumerator1
    Sprite Size Max=<3 5>
    Sprite Size Min=<2 3>
    Sprite Size Mode=::NewEnumerator2
    UsePositionOffset=false
    Write Material Random=true
   ) {module=InitializeParticle002}
   &AddVelocity(
    Cone Angle=90
    Cone Axis=&MakeCustomVectorFromBool(
     Bool=&SetBoolByIntComparison(A=$Particles.PartitionParticles.Partition, B=0, Comparison Type=::NewEnumerator0)
     True Vector=<-1 0 0>
     False Vector=<1 0 0>
    )
    Module.UseConeFalloff=false
    Module.UseVelDistribution=false
    Override Seed=false
    Velocity Mode=::NewEnumerator2
    Velocity Speed=&RandomRangeFloat(Minimum=750, Maximum=1300)
   )
  ]
  particleUpdateScript [
   &ParticleState()
   &InheritSourceMovement(Applied Position Scale=<0.5 0.5 0.5>, ApplyForce=false, ApplyPosition=true, ApplyPositionFalloff=false)
   /Niagara/Modules/Update/Forces/CurlNoiseForce(Noise Frequency=0.3, Noise Strength=400, Pan Noise=false)
   &ScaleColor(
    Scale Alpha=&FloatFromCurve(FloatCurve=DataInterface() {propertyValues {Curve {keys [<at 0 1>,<at 1 0>]}}})
    Scale RGB=&Multiply_VectorByFloat(
     Vector=<1 1 1>
     Float=&FloatFromCurve(
      FloatCurve=DataInterface() {propertyValues {Curve {keys [
          {interpMode=RCIM_Cubic, tangentMode=RCTM_User, time=0, value=0}
          {interpMode=RCIM_Cubic, tangentMode=RCTM_User, time=0.5, value=1}
          {tangentMode=RCTM_User, time=1, value=0}
         ]}}}
      Scale Curve=5
     )
    )
    ScaleA=true
    ScaleRGB=true
   )
   &Drag(
    Drag=&FloatFromCurve(
     FloatCurve=DataInterface() {propertyValues {Curve {keys [
         {arriveTangent=4.1785, interpMode=RCIM_Cubic, leaveTangent=4.1785, tangentMode=RCTM_User, time=0, value=0}
         {arriveTangent=-0.0405, interpMode=RCIM_Cubic, leaveTangent=-0.0405, tangentMode=RCTM_User, time=1, value=1}
        ]}}}
     Scale Curve=13
    )
    Use Linear Drag=true
    Use Rotational Drag=false
   )
   &CameraOffset(Camera Offset Amount=10)
   &SolveForcesAndVelocity(Clamp Velocity=false, Limit Acceleration=false)
   &ScaleSpriteSizeBySpeed(Max Scale Factor=<0.75 10>)
   &ScaleSpriteSize(Scale Factor=&Vector2DFromFloat(Value=&FloatFromCurve(FloatCurve=DataInterface() {propertyValues {Curve {keys [
          {tangentMode=RCTM_User, time=0, value=1}
          {arriveTangent=-0.0405, interpMode=RCIM_Cubic, leaveTangent=-0.0405, tangentMode=RCTM_User, time=0.7009, value=1}
          {arriveTangent=-8.0089, interpMode=RCIM_Cubic, leaveTangent=-8.0089, tangentMode=RCTM_User, time=1, value=0}
         ]}}}))) {module=ScaleSpriteSize001}
  ]
  rendererClasses [&NiagaraSpriteRendererProperties]
  renderers [{rendererClass=&NiagaraSpriteRendererProperties}]
 }
 Round_Flare {
  emitterUpdateScript [
   &EmitterState()
   &SpawnBurst_Instantaneous(Use Spawn Probability=false, UseLoopCountLimit=false)
  ]
  particleSpawnScript [
   &InitializeParticle(
    Color=<rgba 1 0.3149 0 1>
    Lifetime=0.3
    Position Offset=<0 -10 0>
    Sprite Size Mode=::NewEnumerator3
    Uniform Sprite Size=300
    UsePositionOffset=true
    Write Material Random=true
   )
   &CameraOffset(Camera Offset Amount=5)
  ]
  particleUpdateScript [
   &ParticleState()
   &ScaleSpriteSize(Uniform Curve Sprite Scale=DataInterface() {propertyValues {Curve {keys [
        {interpMode=RCIM_Cubic, tangentMode=RCTM_User, time=0, value=0}
        {interpMode=RCIM_Cubic, tangentMode=RCTM_User, time=0.5, value=1}
        {tangentMode=RCTM_User, time=1, value=0}
       ]}}})
   &ScaleColor(
    Scale Alpha=&FloatFromCurve(FloatCurve=DataInterface() {propertyValues {Curve {keys [<at 0.5 1 RCIM_Cubic>,<at 0.95 0 RCIM_Cubic>]}}})
    ScaleA=true
    ScaleRGB=true
   )
   &DynamicMaterialParameters(
    Index 0 Param 1=&FloatFromCurve(FloatCurve=DataInterface() {propertyValues {Curve {keys [<at 0 1>,<at 0.5 0>]}}}, Scale Curve=0.9)
    Index 0 Param 2=0.25
    Index 0 Param 3=1
    Index 0 Param 4=1
    Module.Param0WriteEnabled=true
    Module.Param1WriteEnabled=true
    Module.Param2WriteEnabled=true
    Module.Param3WriteEnabled=true
   ) {enabled=false}
  ]
  rendererClasses [&NiagaraSpriteRendererProperties, &NiagaraMeshRendererProperties]
  renderers [
   {rendererClass=&NiagaraSpriteRendererProperties}
   {rendererClass=&NiagaraMeshRendererProperties, rendererIndex=1}
  ]
 }
}
summary {
 systemName=NS_Emote_MiddleSock_Deploy
 emitters [
  {emitterName=Round_Flare, rendererClasses [&NiagaraSpriteRendererProperties, &NiagaraMeshRendererProperties]}
  {emitterName=Beams, rendererClasses [&NiagaraSpriteRendererProperties]}
 ]
 userVariables []
}
