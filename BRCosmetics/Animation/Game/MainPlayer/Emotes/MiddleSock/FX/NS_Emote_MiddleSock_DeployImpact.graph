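# Niagara burst effect for the deploy impact in the MiddleSock player emote, built from dust, rock and smoke emitters.
# DustSpikes001 bursts short-lived scaling, fading meshes, and SmokeWorld adds small animated sub-image smoke sprites.
# RocksCollision spawns tumbling mesh debris gated by User.PhysMatRemapped, with a DoOnce on speed and age switching collision off.
# User.bIsInWater feeds only a disabled spawn module, and both mesh emitters use distance culling.
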NiagaraSystem /BRCosmetics/Animation/Game/MainPlayer/Emotes/MiddleSock/FX/NS_Emote_MiddleSock_DeployImpact
dependencies {usedDataInterfaces [&NiagaraDataInterfaceCurve], usedDynamicInputs [], usedModules [&SystemState]}
emitters {
 DustSpikes001 {
  emitterUpdateScript [
   &EmitterState(Enable Distance Culling=true, MaxDistance=6000, Module.bUseMaxDistance=true, Module.bUseMinDistance=false, Scalability Mode=::NewEnumerator1)
   &SpawnBurst_Instantaneous(Spawn Count=2, Use Spawn Probability=false, UseLoopCountLimit=false)
   &SpawnBurst_Instantaneous(Spawn Count=2, Spawn Group=1, Spawn Time=0.05, Use Spawn Probability=false, UseLoopCountLimit=false) {enabled=false, module=SpawnBurst_Instantaneous001}
   &SpawnBurst_Instantaneous(Spawn Count=2, Spawn Group=1, Spawn Time=0.1, Use Spawn Probability=false, UseLoopCountLimit=false) {enabled=false, module=SpawnBurst_Instantaneous002}
  ]
  particleSpawnScript [
   &InitializeParticle(
    Color=<rgba 0.1332 0.0853 0.0548 1>
    Lifetime=0.1
    Mesh Scale Mode=::NewEnumerator1
    Position Offset Coordinate Space=::Local
    Sprite Rotation Mode=::NewEnumerator2
    Sprite Size Max=<350 350>
    Sprite Size Min=<200 200>
    Sprite Size Mode=::NewEnumerator2
    UsePositionOffset=true
    Write Material Random=true
   )
   &ShapeLocation(MaskSpawnGroup=false, Override Seed=false, Sphere Radius=5, UseOwnerScale=true)
   &AddVelocity(
    Module.Use Falloff Distance=false
    Origin Offset=<0 0 -75>
    Override Seed=false
    Velocity Mode=::NewEnumerator1
    Velocity Speed=&RandomRangeFloat(Minimum=100, Maximum=200)
   )
   &CameraOffset(Camera Offset Amount=5)
  ]
  particleUpdateScript [
   &ParticleState()
   &GravityForce(Gravity=&MaskVectorBySpawnGroup(Masked Vector=<0 0 0>, Passthrough Vector=<0 0 -960>))
   &Drag(
    Drag=&MaskFloatBySpawnGroup(Masked Float=4, Passthrough Float=1)
    Use Linear Drag=true
    Use Rotational Drag=false
   )
   &ScaleMeshSize(Scale Factor=&VectorFromFloat(Value=&FloatFromCurve(FloatCurve=DataInterface() {propertyValues {Curve {keys [
          {arriveTangent=0.1176, interpMode=RCIM_Cubic, leaveTangent=0.1176, tangentMode=RCTM_User, time=0, value=0.1}
          {arriveTangent=2.0967, interpMode=RCIM_Cubic, leaveTangent=2.0967, tangentMode=RCTM_User, time=1, value=1}
         ]}}})))
   &ScaleColor(
    Scale Alpha=&FloatFromCurve(
     FloatCurve=DataInterface() {propertyValues {Curve {keys [<at 0 1 RCIM_Cubic>, {arriveTangent=-1.4158, interpMode=RCIM_Cubic, leaveTangent=-1.4158, tangentMode=RCTM_User, time=1, value=0}]}}}
     Scale Curve=2
    )
    ScaleA=true
    ScaleRGB=false
   )
   &SolveForcesAndVelocity(Clamp Velocity=false, Limit Acceleration=false)
   &DynamicMaterialParameters(
    Index 0 Param 4=0.5
    Module.Index0Param0Write=true
    Module.Index0Param1Write=true
    Module.Index0Param2Write=true
    Module.Index0Param3Write=true
   )
  ]
  rendererClasses [&NiagaraMeshRendererProperties]
  renderers [{rendererClass=&NiagaraMeshRendererProperties, rendererIndex=0}]
 }
 RocksCollision {
  emitterUpdateScript [
   &EmitterState(Enable Distance Culling=true, MaxDistance=3500, Module.bUseMaxDistance=true, Module.bUseMinDistance=false, Scalability Mode=::NewEnumerator1)
   &SpawnBurst_Instantaneous(
    Spawn Count=&MakeCustomIntFromBool(
     Bool=&SetBoolByIntComparison(A=$User.PhysMatRemapped, B=0, Comparison Type=::NewEnumerator4)
     True INT=5
     False INT=0
    )
    Use Spawn Probability=false
    UseLoopCountLimit=false
   )
  ]
  particleSpawnScript [
   &InitializeParticle(
    Color Maximum=<rgba 0.0435 0.0257 0.0137 1>
    Color Minimum=<rgba 0.185 0.1208 0.0784 1>
    Color Mode=::NewEnumerator2
    Lifetime Max=1
    Lifetime Min=0.7
    Lifetime Mode=::NewEnumerator1
    Mesh Scale Mode=::NewEnumerator1
    Mesh Uniform Scale Max=0.6
    Mesh Uniform Scale Min=0.3
    Position Offset=<0 0 5>
    UsePositionOffset=true
    Write Material Random=true
   )
   &InitialMeshOrientation(
    Rotation=&RandomRangeVector(Minimum=<0 0 0>, Maximum=<1 1 1>)
    Use Orientation Vector=true
    Use Rotation Vector=true
   )
   &ShapeLocation(
    Cylinder Height=5
    Cylinder Height Midpoint=0
    Cylinder Radius=5
    MaskSpawnGroup=false
    Override Seed=false
    Shape Primitive=::NewEnumerator1
    Surface Only=true
    Use Endcaps In Surface Only Mode=false
    UseOwnerScale=true
   )
   &AddVelocity(
    Module.Use Falloff Distance=false
    Origin Offset=<0 0 -60>
    Override Seed=false
    Velocity Mode=::NewEnumerator1
    Velocity Speed=&RandomRangeFloat(Minimum=300, Maximum=500)
   )
   &DynamicMaterialParameters(
    Index 3 Param 1=4
    Index 3 Param 2=0.2
    Module.Param0WriteEnabled=false
    Module.Param0WriteEnabled3=true
    Module.Param1WriteEnabled=false
    Module.Param1WriteEnabled3=true
    Module.Param2WriteEnabled=false
    Module.Param2WriteEnabled3=true
    Module.Param3WriteEnabled=false
    Module.Param3WriteEnabled3=true
    Write Parameter Index 3=true
   )
  ]
  particleUpdateScript [
   &ParticleState()
   &GravityForce(Gravity=<0 0 -1200>)
   &DoOnce(Trigger Condition:NiagaraBool=&Boolean_AndOperation(
     Bool1=&SetBoolByFloatComparison(A=&VectorLength(Vector=$Particles.Velocity), B=150, Comparison Type=::NewEnumerator6)
     Bool2=&SetBoolByFloatComparison(A=$Particles.Age, B=0.5)
    ))
   &Collision(
    Advanced Aging Rate=1
    Age Colliding Particles=false
    CPU Collision Trace Channel=ECC_Visibility
    Collision Enabled=&Boolean_NotOperation(Bool=$Particles.DoOnce.TriggeredOnce)
    Friction=0.5
    Friction During a Bounce=0.3
    Module.Control Rest State Via Surface Normal=false
    Module.EnableMaxCollisionCount=false
    Module.Randomize Collision Normal Vector=true
    Particle Radius Scale=2
    Radius Calculation Type=::NewEnumerator1
    Randomize Collision Normal=0.25
    Restitution=0.4
   )
   &MeshRotationForce(Rotation=&RandomRangeVector(Minimum=<-20 -20 -20>, Maximum=<20 20 20>))
   &SolveForcesAndVelocity(Clamp Velocity=false, Limit Acceleration=false)
   &ScaleMeshSize(Scale Factor=&VectorFromFloat(Value=&FloatFromCurve(FloatCurve=DataInterface() {propertyValues {Curve {keys [
          {arriveTangentWeight=0.2357, interpMode=RCIM_Cubic, leaveTangentWeight=0.2357, tangentMode=RCTM_User, time=0.5, value=1}
          {arriveTangent=-4.4912, interpMode=RCIM_Cubic, leaveTangent=-4.4912, tangentMode=RCTM_User, time=1, value=0}
         ]}}})))
  ]
  rendererClasses [&NiagaraMeshRendererProperties]
  renderers [{rendererClass=&NiagaraMeshRendererProperties, rendererIndex=0}]
 }
 SmokeWorld {
  emitterUpdateScript [
   &EmitterState()
   &SpawnBurst_Instantaneous(Spawn Count=3, Use Spawn Probability=false, UseLoopCountLimit=false)
   &SpawnPerUnit(
    Module.Use Max Movement Threshold=true
    Module.Use Movement Tolerance=true
    Movement Tolerance=1
    Spawn Per Unit:NiagaraFloat=&Multiply_Float(A=&MakeCustomFloatFromBool(Bool=$User.bIsInWater, True Float=0, False Float=1), B=8)
    Use Spawn Probability=false
   ) {enabled=false}
   &SpawnRate(SpawnRate=3, Use Spawn Probability=false) {enabled=false}
  ]
  particleSpawnScript [
   /Niagara/Modules/Spawn/Initialization/InitializeParticle(
    Color=&UniformRangedLinearColor(Minimum=<rgba 0.478 0.3791 0.2199 0.1>, Maximum=<rgba 0.781 0.7271 0.5498 0.15>)
    Lifetime=&UniformRangedFloat(Minimum=0.3, Maximum=0.5)
    Module.Write Color=true
    Module.Write Lifetime=true
    Module.Write Mass=true
    Module.Write Material Random=true
    Module.Write Position=false
    Module.Write Scale=false
    Module.Write SpriteRotation=true
    Module.Write SpriteSize=true
    Position=&SimulationPosition()
    Sprite Rotation=&UniformRangedFloat(Minimum=0, Maximum=360)
    Sprite Size=&Vector2DFromFloat(Value=&UniformRangedFloat(Minimum=10, Maximum=20))
   )
   &ShapeLocation(MaskSpawnGroup=false, Offset=<5 0 0>, Offset Mode=::NewEnumerator0, Override Seed=false, Sphere Radius=3, UseOwnerScale=true)
   &AddVelocity(Module.Use Falloff Distance=false, Origin Offset=<1 0 0>, Override Seed=false, Velocity Mode=::NewEnumerator1, Velocity Speed=35) {module=AddVelocity001}
   &AddVelocity(Velocity=<0 0 100>) {enabled=false}
  ]
  particleUpdateScript [
   &ParticleState()
   &Drag(
    Drag=&FN_FrontEndFloatModifier(Gameplay=5, Front End=1)
    Use Linear Drag=true
    Use Rotational Drag=false
   ) {enabled=false}
   &SolveForcesAndVelocity(Clamp Velocity=false, Limit Acceleration=false)
   &ScaleSpriteSize(Scale Factor=&Vector2DFromFloat(Value=&FloatFromCurve(FloatCurve=DataInterface() {propertyValues {Curve {keys [
          {arriveTangent=1.491, interpMode=RCIM_Cubic, leaveTangent=1.491, tangentMode=RCTM_User, time=-0.0046, value=0.6}
          <at 1.0011 3.4809 RCIM_Cubic>
         ]}}})))
   /SetVariables(Particles.SubImageIndex=&FloatFromCurve(FloatCurve=DataInterface() {propertyValues {Curve {keys [
         {arriveTangent=29.8623, interpMode=RCIM_Cubic, leaveTangent=29.8623, tangentMode=RCTM_User, time=-0.0046, value=5.0556}
         <at 1.0011 22.1689 RCIM_Cubic>
        ]}}})) {isSetParameters=true}
   &DynamicMaterialParameters(
    Index 0 Param 1=&FloatFromCurve(FloatCurve=DataInterface() {propertyValues {Curve {keys [<at -0.0046 1.0113 RCIM_Cubic>,<at 1.0011 0.002>]}}})
    Index 0 Param 4=&FloatFromCurve(FloatCurve=DataInterface() {propertyValues {Curve {keys [<at 0 0 RCIM_Cubic>,<at 1 0.5>]}}})
    Module.Param0WriteEnabled=true
    Module.Param1WriteEnabled=false
    Module.Param2WriteEnabled=false
    Module.Param3WriteEnabled=true
   ) {module=DynamicMatParams}
   &CameraOffset(Camera Offset Amount=5)
   &ScaleColor(
    Scale Mode=::NewEnumerator1
    Scale RGBA=&MakeVector4_FromLinearColor(LinearColor=&FNTimeOfDayColor(
      Night Color=<rgba 0.4115 0.4079 0.3609 0.01>
      Morning Color=<rgba 0.781 0.7271 0.5498 0.1>
      Day Color=<rgba 0.781 0.7271 0.5498 0.1>
      Evening Color=<rgba 0.4115 0.4079 0.3609 0.01>
     ))
    ScaleRGBA=true
   ) {enabled=false}
   &ScaleColor(
    Scale Alpha=&FloatFromCurve(FloatCurve=DataInterface() {propertyValues {Curve {keys [
         {interpMode=RCIM_Cubic, tangentMode=RCTM_User, time=0, value=0}
         {interpMode=RCIM_Cubic, tangentMode=RCTM_User, time=0.5, value=1}
         {tangentMode=RCTM_User, time=1, value=0}
        ]}}})
    ScaleA=true
    ScaleRGB=true
   ) {module=ScaleColor001}
  ]
  rendererClasses [&NiagaraSpriteRendererProperties]
  renderers [{rendererClass=&NiagaraSpriteRendererProperties}]
 }
}
summary {
 systemName=NS_Emote_MiddleSock_DeployImpact
 emitters [
  {emitterName=DustSpikes001, rendererClasses [&NiagaraMeshRendererProperties]}
  {emitterName=RocksCollision, rendererClasses [&NiagaraMeshRendererProperties]}
  {emitterName=SmokeWorld, rendererClasses [&NiagaraSpriteRendererProperties]}
 ]
 userVariables [
  User.bIsInWater {type=NiagaraBool}
  User.PhysMatRemapped {type=NiagaraInt32}
 ]
}
